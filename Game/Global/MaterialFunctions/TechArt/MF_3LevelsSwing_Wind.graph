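# Material function that adds three-level hierarchical wind swing, likely for foliage, to material attributes through World Position Offset, behind a static switch.
# Threshold masks on a selectable vertex color channel (alpha by default) drive three RotateAboutAxis stages, with pivots derived from UV1/UV2 or overrides.
# Wind actor direction sets swing axis and a facing-based strength limiter, scaling procedural sine/cosine gusts.
# A debug switch shows pivots and level masks as emissive.
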MaterialFunction /Game/Global/MaterialFunctions/TechArt/MF_3LevelsSwing_Wind
graph {nodes [
  Constant() {r=1}
  n1 = LocalPosition()
  n2 = MaterialFunctionCall() {materialFunction=&ObjectScale}
  Divide(A=@n1.XYZ, B=@n2.Scale XYZ)
  n3 = GetMaterialAttributes(None=FunctionInput() {inputName=MA, inputType=FunctionInput_MaterialAttributes, sortPriority=-100}) {attributeGetTypes [#1, #2]}
  n4 = Add(A=NamedRerouteUsage(), B=NamedRerouteUsage())
  n5 = NamedRerouteUsage()
  n6 = Add(A=@n3.WorldPositionOffset, B=Transform(None=Add(A=@n4, B=@n5)) {transformSourceType=TRANSFORMSOURCE_Local})
  n7 = SetMaterialAttributes!(MaterialAttributes, EmissiveColor=NamedRerouteUsage(), WorldPositionOffset=@n6) {attributeSetTypes [#3, #1]}
  n8 = SetMaterialAttributes!(MaterialAttributes=@n3.MaterialAttributes, WorldPositionOffset=@n6, Normal=@n3.Normal) {attributeSetTypes [#1, #2]}
  FunctionOutput(None=StaticSwitchParameter(
    True=StaticSwitchParameter(True=@n7, False=@n8) {group=Three Levels Swing - Debug, parameterName=Debug Swing Wind, sortPriority=0}
    False=@n3.MaterialAttributes
   ) {group=Three Levels Swing - Wind, parameterName=Use 3 Level Swing Wind, sortPriority=0})
  NamedRerouteDeclaration(None=Multiply(A=@n1.XYZ, B=@n2.Scale XYZ)) {name=Local Position, nodeColor=<rgba 1 0.1 0.1 1>}
  NamedRerouteDeclaration(None=Multiply(A=Time(), B=ScalarParameter() {defaultValue=0.1, group=Three Levels Swing - Wind, parameterName=Timeline Speed})) {name=Timeline, nodeColor=<rgba 0.875 0 1 1>}
  n9 = ActorPositionWS()
  NamedRerouteDeclaration(None=Multiply(
    A=TransformPosition(None=@n9) {transformSourceType=TRANSFORMPOSSOURCE_World}
    B=MaterialFunctionCall() {materialFunction=&ObjectScale}.Scale XYZ
   )) {name=Actor Offset, nodeColor=<rgba 0.8516 0 1 1>}
  n10 = VertexColor()
  NamedRerouteDeclaration(None=ChannelMaskParameter(None=AppendVector(A=@n10, B=@n10)) {maskChannel=Alpha, parameterName=Levels Mask Channel, defaultValue=<rgba 0 0 0 1>}.R) {name=Levels Mask Channel, nodeColor=<rgba 1 0 0.8437 1>}
  n11 = Constant() {r=1}
  n12 = NamedRerouteDeclaration(None=If(
    A=NamedRerouteUsage()
    B=ScalarParameter() {defaultValue=0.1, group=Three Levels Swing - Wind, parameterName=First Level Mask Value, sortPriority=400}
    "A > B"=@n11
    "A == B"=@n11
    "A < B"=Constant()
   ) {equalsThreshold=0.05}) {name=First Level Mask, nodeColor=<rgba 0.5 0 1 1>}
  n13 = ComponentMask(None=NamedRerouteUsage()) {r=true}
  n14 = NamedRerouteUsage()
  n15 = Add(A=NamedRerouteUsage(), B=NamedRerouteUsage())
  NamedRerouteDeclaration(None=RotateAboutAxis(
    NormalizedRotationAxis=NamedRerouteUsage()
    RotationAngle=Multiply(A=@n12, B=@n13)
    PivotPoint=NamedRerouteUsage()
    Position=Add(A=@n14, B=@n15)
   ) {period=3.14}) {name=First Rotation, nodeColor=<rgba 1 0 0.7969 1>}
  n16 = Constant() {r=1}
  n17 = NamedRerouteDeclaration(None=If(
    A=NamedRerouteUsage()
    B=ScalarParameter() {defaultValue=0.5, group=Three Levels Swing - Wind, parameterName=Middle Level Mask Value, sortPriority=401}
    "A > B"=@n16
    "A == B"=@n16
    "A < B"=Constant()
   ) {equalsThreshold=0.05}) {name=Middle Level Mask, nodeColor=<rgba 0.5 0 1 1>}
  n18 = ComponentMask(None=NamedRerouteUsage()) {g=true}
  n19 = NamedRerouteUsage()
  n20 = NamedRerouteUsage()
  NamedRerouteDeclaration(None=RotateAboutAxis(
    NormalizedRotationAxis=NamedRerouteUsage()
    RotationAngle=Multiply(A=@n17, B=@n18)
    PivotPoint=NamedRerouteUsage()
    Position=Add(A=@n19, B=@n20)
   ) {period=3.14}) {name=Middle Rotation, nodeColor=<rgba 0 1 0.8828 1>}
  n21 = Constant() {r=1}
  n22 = NamedRerouteDeclaration(None=If(
    A=NamedRerouteUsage()
    B=ScalarParameter() {defaultValue=1, group=Three Levels Swing - Wind, parameterName=Last Level Mask Value, sortPriority=402}
    "A > B"=@n21
    "A == B"=@n21
    "A < B"=Constant()
   ) {equalsThreshold=0.05}) {name=Last Level Mask, nodeColor=<rgba 0.5 0 1 1>}
  n23 = ComponentMask(None=NamedRerouteUsage()) {b=true}
  NamedRerouteDeclaration(None=RotateAboutAxis(
    NormalizedRotationAxis=NamedRerouteUsage()
    RotationAngle=Multiply(A=@n22, B=@n23)
    PivotPoint=NamedRerouteUsage()
    Position=NamedRerouteUsage()
   ) {period=3.14}) {name=Last Rotation, nodeColor=<rgba 0.9688 1 0 1>}
  n24 = VectorParameter() {
   group=Three Levels Swing - Intensity Limiter
   parameterName=Direction Limit Multiplier
   channelNames {a=Enable Invert Forward, b=Limit Multiplier -Y, g=Limit Multiplier +Y, r=Limit Multiplier X}
   defaultValue=<rgba 0.5 1 0.2 0>
  }
  n25 = VectorParameter() {
   group=Three Levels Swing - Intensity Limiter
   parameterName=Swing Strength
   channelNames {a="", b=Opposite Strength -Y, g=Opposite Strength +Y, r=Opposite Strength X}
  }
  n26 = AppendVector(A=@n24.Limit Multiplier +Y, B=@n25.Opposite Strength +Y)
  n27 = AppendVector(A=@n24.Limit Multiplier X, B=@n25.Opposite Strength X)
  n28 = AppendVector(A=@n24.Limit Multiplier -Y, B=@n25.Opposite Strength -Y)
  n29 = Normalize(VectorInput=MaterialFunctionCall!(Position, IsTrees, IsGrass, IsMediumPlant, IsMediumPlant_2024Anim, IsMediumPlant2ndLevel, MediumPlant_BasePivotPostion) {materialFunction=/Game/Environments/Asteria/Foliage/Wind/MF_WindActorData}.WindI_Direction)
  n30 = VertexColor()
  n31 = MaterialFunctionCall(X=@n30.R, Y=@n30.R, Z=Constant() {r=0.5}) {materialFunction=&MakeFloat3}
  n32 = VectorParameter() {group=Three Levels Swing - Intensity Limiter, parameterName=Override Limiter Forward Vector, sortPriority=103, defaultValue=<rgba 0 1 0 0>}
  n33 = FunctionInput(Preview=LinearInterpolate(
    A=ConstantBiasScale(None=@n31.Result) {bias=-0.5, scale=2}
    B=@n32.RGB
    Alpha=ScalarParameter() {group=Three Levels Swing - Intensity Limiter, parameterName=Override Limiter Axis, sortPriority=103}
   )) {bUsePreviewValueAsDefault=true, description=Forward Vectir in Local Space, inputName=Custom Forward Vector, sortPriority=103}
  n34 = ConstantBiasScale(None=DotProduct(A=Transform(None=@n29) {transformSourceType=TRANSFORMSOURCE_World, transformType=TRANSFORM_Local}, B=@n33))
  NamedRerouteDeclaration(None=ComponentMask(None=MaterialFunctionCall!(1=@n26, 2=@n27, 3=@n28, 4, A=@n34) {materialFunction=/Engine/Functions/Engine_MaterialFunctions02/Lerp_Multiple_Float2}.Lerp 3 Inputs) {r=true}) {name=Limiter, nodeColor=<rgba 1 0 0.375 1>}
  n35 = TextureCoordinate() {coordinateIndex=2}
  n36 = AppendVector(A=TextureCoordinate() {coordinateIndex=1}, B=ComponentMask(None=@n35) {r=true})
  n37 = FunctionInput(Preview=LinearInterpolate(
    A=@n36
    B=VectorParameter() {group=Three Levels Swing - Wind, parameterName=Override Middle Pivot Local Position, sortPriority=101}.RGB
    Alpha=ScalarParameter() {group=Three Levels Swing - Wind, parameterName=Override Middle Pivot, sortPriority=101}
   )) {bUsePreviewValueAsDefault=true, description=Pivot in Local Space, inputName=Custom Middle Pivot, sortPriority=101}
  NamedRerouteDeclaration(None=ComponentMask(None=@n37) {g=true, r=true}) {name=Local XY Pivot, nodeColor=<rgba 0 0.6484 1 1>}
  n38 = CrossProduct(
   A=Constant3Vector() {constant=<rgba 0 0 1 0>}
   B=Transform(None=NamedRerouteUsage()) {transformSourceType=TRANSFORMSOURCE_World, transformType=TRANSFORM_Local}
  )
  NamedRerouteDeclaration(None=FunctionInput(Preview=Normalize(VectorInput=@n38)) {bUsePreviewValueAsDefault=true, description=Local Space, inputName=Custom Rotation Axis, sortPriority=700}) {name=Rotation Axis, nodeColor=<rgba 0 1 0.5078 1>}
  n39 = StaticSwitchParameter(True=Add(A=NamedRerouteUsage(), B=@n37), False=@n37) {group=Three Levels Swing - Wind, parameterName=Use Actor as Origin, sortPriority=5}
  NamedRerouteDeclaration(None=Divide(
    A=@n39
    B=MaterialFunctionCall() {materialFunction=&ObjectScale}.Scale XYZ
   )) {name=Middle Pivot, nodeColor=<rgba 0 1 1 1>}
  n40 = MaterialFunctionCall(Z=ComponentMask(None=@n35) {g=true}) {materialFunction=&MakeFloat3}
  n41 = FunctionInput(Preview=LinearInterpolate(
    A=Subtract(A=@n36, B=@n40.Result)
    B=VectorParameter() {group=Three Levels Swing - Wind, parameterName=Override First Pivot Local Position, sortPriority=100}.RGB
    Alpha=ScalarParameter() {group=Three Levels Swing - Wind, parameterName=Override First Pivot, sortPriority=100}
   )) {bUsePreviewValueAsDefault=true, description=Pivot in Local Space, inputName=Custom First Pivot, sortPriority=100}
  n42 = Add(A=NamedRerouteUsage(), B=@n41)
  NamedRerouteDeclaration(None=Divide(
    A=StaticSwitchParameter(True=@n42, False=@n41) {group=Three Levels Swing - Wind, parameterName=Use Actor as Origin, sortPriority=5}
    B=MaterialFunctionCall() {materialFunction=&ObjectScale}.Scale XYZ
   )) {name=First Pivot, nodeColor=<rgba 0 1 1 1>}
  n43 = VectorParameter() {group=Three Levels Swing - Wind, parameterName=Last Pivot Offset}
  n44 = FunctionInput(Preview=LinearInterpolate(
    A=Add(A=@n39, B=@n43.RGB)
    B=VectorParameter() {group=Three Levels Swing - Wind, parameterName=Override Last Pivot Local Position, sortPriority=102}.RGB
    Alpha=ScalarParameter() {group=Three Levels Swing - Wind, parameterName=Override Last Pivot, sortPriority=102}
   )) {bUsePreviewValueAsDefault=true, description=Pivot in Local Space, inputName=Custom Last Pivot, sortPriority=102}
  n45 = Add(A=NamedRerouteUsage(), B=@n44)
  NamedRerouteDeclaration(None=Divide(
    A=StaticSwitchParameter(True=@n45, False=@n44) {group=Three Levels Swing - Wind, parameterName=Use Actor as Origin, sortPriority=5}
    B=MaterialFunctionCall() {materialFunction=&ObjectScale}.Scale XYZ
   )) {name=Last Pivot, nodeColor=<rgba 0 1 1 1>}
  NamedRerouteDeclaration(None=@n29) {name=Wind Direction, nodeColor=<rgba 0 1 0.1094 1>}
  n46 = FunctionInput(Preview=NamedRerouteUsage()) {bUsePreviewValueAsDefault=true, inputName=Wind Pattern, sortPriority=800}
  n47 = VectorParameter() {
   group=Three Levels Swing - Wind
   parameterName=Wind Intensity
   channelNames {a="", b=Level 3, g=Level 2, r=Level 1}
   defaultValue=<rgba 0.1 0.1 0.15 1>
  }
  NamedRerouteDeclaration(None=Multiply(A=Multiply(A=@n46, B=@n47), B=NamedRerouteUsage())) {name=Wind, nodeColor=<rgba 1 0 0.5391 1>}
  n48 = NamedRerouteUsage()
  n49 = TransformPosition(None=LocalPosition() {localOrigin=Primitive}.XYZ) {transformType=TRANSFORMPOSSOURCE_World}
  n50 = Add(
   A=MaterialFunctionCall!(
    Sphere Position=TransformPosition(None=@n48) {transformType=TRANSFORMPOSSOURCE_World}
    Sphere Radius=ScalarParameter() {defaultValue=1, group=Three Levels Swing - Debug, parameterName=Debug Pivot Radius}
    RayDirection
    World position=@n49
   ) {materialFunction=/Engine/Functions/Engine_MaterialFunctions02/RayTracedSphere}.Ray Hit Bool
   B=Multiply(A=NamedRerouteUsage()) {constB=0.2}
  )
  n51 = Constant3Vector() {constant=<rgba 1 0 0 0>}
  n52 = TransformPosition(None=NamedRerouteUsage()) {transformType=TRANSFORMPOSSOURCE_World}
  n53 = ScalarParameter() {defaultValue=1, group=Three Levels Swing - Debug, parameterName=Debug Pivot Radius}
  n54 = NamedRerouteUsage()
  n55 = Multiply(
   A=Add(
    A=MaterialFunctionCall(Sphere Position=@n52, Sphere Radius=@n53, World position=@n49) {materialFunction=&RayTracedSphere}.Ray Hit Bool
    B=Multiply(A=@n54) {constB=0.2}
   )
   B=Constant3Vector() {constant=<rgba 0 1 0 0>}
  )
  n56 = TransformPosition(None=NamedRerouteUsage()) {transformType=TRANSFORMPOSSOURCE_World}
  n57 = ScalarParameter() {defaultValue=1, group=Three Levels Swing - Debug, parameterName=Debug Pivot Radius}
  n58 = NamedRerouteUsage()
  n59 = Multiply(
   A=Add(
    A=MaterialFunctionCall(Sphere Position=@n56, Sphere Radius=@n57, World position=@n49) {materialFunction=&RayTracedSphere}.Ray Hit Bool
    B=Multiply(A=@n58) {constB=0.2}
   )
   B=Constant3Vector() {constant=<rgba 0 0 1 0>}
  )
  NamedRerouteDeclaration(None=Add(A=Multiply(A=@n50, B=@n51), B=Add(A=@n55, B=@n59))) {name=Debug Pivots, nodeColor=<rgba 0 1 0.4844 1>}
  n60 = Add(
   A=Divide(A=NamedRerouteUsage(), B=ScalarParameter() {defaultValue=350, group=Three Levels Swing - Wind, parameterName=Wind Scale})
   B=MaterialFunctionCall(X=NamedRerouteUsage()) {materialFunction=&MakeFloat2}.Result
  )
  n61 = ComponentMask(None=@n60) {r=true}
  n62 = MaterialFunctionCall(X=@n61, Y=Add(A=@n61, B=ComponentMask(None=@n60) {g=true})) {materialFunction=&MakeFloat2}
  n63 = MaterialFunctionCall(X=Sine(None=ComponentMask(None=@n62.Result) {g=true}) {period=0.7}, Y=Cosine(None=ComponentMask(None=@n62.Result) {r=true}) {period=1.1}) {materialFunction=&MakeFloat2}
  n64 = Constant() {r=4}
  n65 = Multiply(A=@n62.Result, B=@n64)
  n66 = Multiply(A=@n62.Result, B=@n64)
  n67 = MaterialFunctionCall(X=Sine(None=ComponentMask(None=@n65) {r=true}) {period=1.3}, Y=Cosine(None=ComponentMask(None=@n66) {g=true}) {period=1.5}) {materialFunction=&MakeFloat2}
  n68 = LinearInterpolate(A=@n60, B=ConstantBiasScale(None=LinearInterpolate(A=@n63.Result, B=@n67.Result) {constAlpha=0.2})) {constAlpha=0.25}
  n69 = ComponentMask(None=Multiply(A=@n68) {constB=0.5}) {r=true}
  n70 = Constant3Vector() {constant=<rgba 0 -0.1 -0.2 0>}
  n71 = ComponentMask(None=@n68) {g=true}
  n72 = Multiply(A=Sine(None=Add(A=@n69, B=@n70)) {period=0.2}, B=Cosine(None=Add(A=@n71, B=@n70)) {period=0.4})
  NamedRerouteDeclaration(None=LinearInterpolate(A=ScalarParameter() {defaultValue=-0.35, group=Three Levels Swing - Wind, parameterName=Swing Amount}, Alpha=ConstantBiasScale(None=@n72))) {name=Wind Procedural Texture, nodeColor=<rgba 0.125 0 1 1>}
 ]}
