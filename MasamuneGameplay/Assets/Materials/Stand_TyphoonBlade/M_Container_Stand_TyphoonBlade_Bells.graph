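# Master material for the Typhoon Blade stand container with bells, built on character layer functions and Material Attributes.
# An optional Blade Hover switch adds noise-driven bob and rotation masked by vertex green, plus fake glow lighting.
# Bells swing through three-level pivot wind with UV-seeded variation.
# Backface-corrected world normals then feed jungle biome and RVT blending, a low-quality mobile simplification, Creative move tool and texture-masked cloth wind.
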Material /MasamuneGameplay/Assets/Materials/Stand_TyphoonBlade/M_Container_Stand_TyphoonBlade_Bells
graph {
 properties {
  bAutomaticallySetUsageInEditor=false
  bTangentSpaceNormal=false
  bUseMaterialAttributes=true
  bUsedWithInstancedStaticMeshes=true
  bUsedWithMeshDeformer=true
  bUsedWithNanite=true
  bUsedWithSkeletalMesh=true
  bUsedWithSplineMeshes=true
  ditheredLODTransition=true
  pixelDepthOffsetMode=PDOM_Legacy
  shadingModel=MSM_FromMaterialExpression
  usedShadingModels=Default Lit | Subsurface | Clear Coat | Subsurface Profile | Cloth | Eye
 }
 nodes [
  n1 = MaterialFunctionCall() {materialFunction=&BaseFN_CharacterFunction_PreFX}
  n2 = MaterialFunctionCall!(SFX V3, In=@n1.Result) {materialFunction=/Game/Global/Base4Layers_2022/Material/MF_Cover}
  n3 = GetMaterialAttributes(None=@n2.Result) {attributeGetTypes [#1]}
  n4 = SetMaterialAttributes!(
   MaterialAttributes=@n3.MaterialAttributes
   World Position Offset=NamedRerouteUsage()
   Emissive Color=Add(A=@n3.Emissive Color, B=NamedRerouteUsage()) {constB=0}
  ) {attributeSetTypes [#2, #1]}
  n5 = MaterialFunctionCall(
   AO=@n1.AO
   Distance Blend=@n1.AO
   Fresnel=@n1.AO
   Gmap=@n1.AO
   Pre Skinned Local Position=@n1.AO
   Material=MaterialFunctionCall(
    PreskinnedPosition=@n1.Pre skinned Local Position
    Gmap=@n1.Pre skinned Local Position
    MaterialAttributes=StaticSwitchParameter(True=@n4, False=@n2.Result) {parameterName=Use Blade Hover}
    AO=@n1.Pre skinned Local Position
   ) {materialFunction=&CharacterSkinFX}.Result
  ) {materialFunction=&BaseFN_CharacterFunction_PostFX}
  n6 = Normalize(VectorInput=AppendVector(A=TextureCoordinate() {coordinateIndex=1}, B=Constant()))
  n7 = CrossProduct(A=Constant3Vector() {constant=<rgba 0 0 1 0>}, B=@n6)
  n8 = Multiply(A=@n7) {constB=-1}
  n9 = Multiply(A=Multiply(A=ScalarParameter() {defaultValue=0.0512, parameterName=Deflection Speed}, B=NamedRerouteUsage()), B=NamedRerouteUsage())
  n10 = ScalarParameter() {defaultValue=0.35, parameterName=Diflection Amount}
  n11 = NamedRerouteUsage()
  n12 = Add(A=Multiply(A=Sine(None=@n9), B=Multiply(A=@n10, B=@n11))) {constB=0.5}
  n13 = Multiply(A=NamedRerouteUsage(), B=Multiply(A=ScalarParameter() {defaultValue=0.5, parameterName=Bell Swing Speed}, B=NamedRerouteUsage()))
  n14 = MaterialFunctionCall!(
   MA=@n5.Result
   Custom First Pivot
   Custom Middle Pivot
   Custom Last Pivot
   Custom Forward Vector
   Custom Rotation Axis=Normalize(VectorInput=MaterialFunctionCall(1=@n8, 2=@n6, 3=@n7, A=@n12) {materialFunction=&Lerp_Multiple_Float3}.Lerp 3 Inputs)
   Wind Pattern=Multiply(A=Sine(None=@n13), B=ScalarParameter() {defaultValue=0.4, parameterName=Bell Wind Amp})
  ) {materialFunction=/Game/Global/MaterialFunctions/TechArt/MF_3LevelsSwing_Wind}
  Composite!(None=@n14.Result) {desc=Transform to World in case we will use RVT after this point, subgraphName=Tangent to World}
  MaterialExpressionReroute_1 = Reroute(None=@n20.RVT H)
  MaterialExpressionReroute_3 = Reroute(None=@n14.Result)
  n15 = GetMaterialAttributes(None=@MaterialExpressionReroute_3.Input) {attributeGetTypes [#3]}
  n16 = Transform(None=@n15.Normal)
  n17 = LinearInterpolate(A=Multiply(A=@n16) {constB=-1}, B=@n16, Alpha=Saturate(None=TwoSidedSign())) {desc="Fixes normals\r\non backfaces"}
  n18 = SetMaterialAttributes!(
   MaterialAttributes=@n15.MaterialAttributes
   Normal=StaticSwitchParameter(True=MaterialProxyReplace(Realtime=@n17, MaterialProxy=@n15.Normal), False=@n17) {parameterName=Force TangentSpaceNormalsFor HLOD ProxyReplace}
  ) {attributeSetTypes [#3]}
  MaterialExpressionReroute_2 = Reroute(None=@n18)
  MaterialExpressionReroute_8 = Reroute(None=@MaterialExpressionReroute_2.Output)
  n19 = MaterialFunctionCall!(AO, RVT_Height=@MaterialExpressionReroute_1."RVT_Height (S)", For Buildings=@MaterialExpressionReroute_8."For Moss (MA)", For Moss) {materialFunction=/Game/Environments/Asteria/Rocks/Materials/MaterialFunctions/MF_JungleBiome}
  MaterialExpressionReroute_0 = Reroute(None=@n19.Buildings Jungl)
  n20 = MaterialFunctionCall!(
   In=@MaterialExpressionReroute_0.Result
   "RVT Height Map (Roughness)"=GetMaterialAttributes(None=@MaterialExpressionReroute_2.Output) {attributeGetTypes [#4]}.Roughness
  ) {materialFunction=/Game/Global/Base4Layers_2022/MaterialFunctions/MF_RVTBlending}
  Composite!(InputPin=@n20.Result) {subgraphName=Mobile Opt}
  Composite!("For Moss (MA)"=@MaterialExpressionReroute_2.Output, "RVT_Height (S)"=@n20.RVT H) {subgraphName="Seasonal "}
  LinearInterpolate(Alpha=NamedRerouteUsage())
  MaterialFunctionCall!(Position, IsTrees, IsGrass, IsMediumPlant, IsMediumPlant_2024Anim, IsMediumPlant2ndLevel, MediumPlant_BasePivotPostion) {materialFunction=/Game/Environments/Asteria/Foliage/Wind/MF_WindActorData}
  MaterialExpressionReroute_7 = Reroute(None=@n20.Result)
  n21 = GetMaterialAttributes(None=@MaterialExpressionReroute_7.InputPin)
  n22 = Constant() {r=0.5}
  n23 = Constant() {r=0.8}
  n24 = ShadingPathSwitch(
   Default=@MaterialExpressionReroute_7.InputPin
   Mobile=QualitySwitch(
    Default=@MaterialExpressionReroute_7.InputPin
    Low=SetMaterialAttributes!(MaterialAttributes=@n21.MaterialAttributes, Specular=@n22, Roughness=@n23) {attributeSetTypes [#5, #4]}
    High=@MaterialExpressionReroute_7.InputPin
    Medium=@MaterialExpressionReroute_7.InputPin
    Epic=@MaterialExpressionReroute_7.InputPin
   )
  )
  MaterialExpressionReroute_6 = Reroute(None=@n24)
  n25 = LocalPosition()
  n26 = Power(Base=Divide(A=ComponentMask(None=@n25.XYZ) {g=true}, B=Constant() {r=60}) {constB=50}) {constExponent=0.5}
  n27 = Constant3Vector() {constant=<rgba 0 0 1 0>}
  MaterialFunctionCall!(
   In=MaterialFunctionCall(MA In=@MaterialExpressionReroute_6.Output) {materialFunction=&MF_CreativeMoveTool}.MA Out
   Custom Mask=Multiply(
    A=TextureSample() {samplerType=SAMPLERTYPE_Masks, texture=/MasamuneGameplay/Assets/Textures/Sharp_TyphoonBlade/T_Sharp_TyphoonBlade_ClothMask}.R
    B=Saturate(None=@n26)
   )
   Custom Rotation Axis=@n27
  ) {materialFunction=/Game/Athena/WinterArt/Materials/MF_WindWPO_v2}
  n28 = MaterialFunctionCall(Float2=TextureCoordinate() {coordinateIndex=1}) {materialFunction=&BreakOutFloat2Components}
  n29 = Add(A=Add(A=@n28.R, B=@n28.R), B=ComponentMask(None=TextureCoordinate() {coordinateIndex=2}) {g=true})
  NamedRerouteDeclaration(None=Add(A=Time(), B=@n29)) {name=Random Time, nodeColor=<rgba 1 0.8438 0 1>}
  n30 = Constant3Vector() {constant=<rgba 0 0 1 0>}
  n31 = Time()
  n32 = Constant() {r=90}
  n33 = MaterialFunctionCall(Float3=MaterialFunctionCall!(UV1=Divide(A=@n31, B=@n32) {constB=10}, Min, Max) {materialFunction=/Game/Athena/WinterArt/Materials/MF_SimpleWindNoise}.Result) {materialFunction=&BreakOutFloat3Components}
  n34 = Multiply(A=Subtract(A=@n33.G) {constB=0.5}, B=ScalarParameter() {defaultValue=0.07, parameterName=MaxRot}) {constB=0.1}
  NamedRerouteDeclaration(None=MaterialFunctionCall(Rotation Axis=@n30, Rotation Angle=@n34) {materialFunction=&FixRotateAboutAxisNormals}.New Tangent Space Vertex Normal) {name=Fix Normal, nodeColor=<rgba 0 0.2734 1 1>}
  NamedRerouteDeclaration(None=ScalarParameter() {parameterName=StormDebug}) {name=Storm, nodeColor=<rgba 0 1 0.4141 1>}
  NamedRerouteDeclaration(None=LinearInterpolate(Alpha=Frac(None=@n29)) {constA=0.96, constB=1.04}) {name=Random Diveation, nodeColor=<rgba 1 0 0.9844 1>}
  n35 = Multiply(
   A=Multiply(A=@n33.R, B=ScalarParameter() {defaultValue=17, parameterName=UpdownDist}) {constB=25}
   B=Constant3Vector() {constant=<rgba 0 0 1 0>}
  )
  n36 = RotateAboutAxis(
   NormalizedRotationAxis=@n30
   RotationAngle=@n34
   PivotPoint=Add(A=ActorPositionWS() {originType=CameraRelative}, B=Constant3Vector() {constant=<rgba 0 0 152 0>})
   Position=WorldPosition() {worldPositionShaderOffset=WPT_CameraRelative}.XYZ
  )
  NamedRerouteDeclaration(None=Multiply(A=Add(A=@n35, B=@n36), B=VertexColor().G)) {name=Hover WPO, nodeColor=<rgba 0.5469 0 1 1>}
  n37 = VertexNormalWS()
  n38 = Saturate(None=Divide(A=LocalPosition().Z, B=ScalarParameter() {defaultValue=200, parameterName=Remap Gradient}))
  n39 = ScalarParameter() {defaultValue=0.3, parameterName=Gradient Exponent}
  n40 = Transform(None=LinearInterpolate(B=VectorParameter() {parameterName=Glow Warp Vert Norm, defaultValue=<rgba 0.5 1.4 1 1>}.RGB, Alpha=Power(Base=@n38, Exp=@n39)) {constA=1}) {transformSourceType=TRANSFORMSOURCE_Local}
  n41 = VectorParameter() {parameterName=Glow Direction, defaultValue=<rgba 0 -1 0.5 1>}
  n42 = DotProduct(
   A=Normalize(VectorInput=Multiply(A=@n37, B=@n40))
   B=Transform(None=Normalize(VectorInput=@n41.RGB)) {transformSourceType=TRANSFORMSOURCE_Local}
  )
  n43 = Multiply(
   A=SmoothStep(
    Min=ScalarParameter() {defaultValue=0.75, parameterName=Sword Glow Remap Min}
    Max=ScalarParameter() {defaultValue=1, parameterName=Sword Glow Remap Max}
    Value=ConstantBiasScale(None=@n42)
   )
   B=ScalarParameter() {defaultValue=2, parameterName=Sword Glow Multiplier}
  )
  n44 = Transform(None=Constant3Vector() {constant=<rgba 0 0 1 0>}) {transformSourceType=TRANSFORMSOURCE_Local}
  n45 = CameraVectorWS()
  n46 = Power(Base=Abs(None=DotProduct(A=@n44, B=@n45)))
  n47 = Multiply(
   A=Fresnel(ExponentIn=ScalarParameter() {defaultValue=1.5, parameterName=Sword Fresnel Exponent}) {baseReflectFraction=0}
   B=Multiply(A=ScalarParameter() {defaultValue=1, parameterName=Sword Fresnel Multiplier}, B=OneMinus(None=@n46))
  )
  NamedRerouteDeclaration(None=Multiply(A=Max(A=@n43, B=@n47), B=VectorParameter() {parameterName=Sword Glow Color, defaultValue=<rgba 0.65 0.3991 0.1249 1>}.RGB)) {name=Fake Ligting, nodeColor=<rgba 0 1 0.625 1>}
  PinBase() {reroutePins [Input {expression=&self:MaterialExpressionReroute_3}]}
  PinBase!(Output=@n18) {reroutePins [Output {expression=&self:MaterialExpressionReroute_2}]}
  PinBase!() {reroutePins [
    "For Moss (MA)" {expression=&self:MaterialExpressionReroute_8}
    "RVT_Height (S)" {expression=&self:MaterialExpressionReroute_1}
   ]}
  PinBase!(Result=@n19.Buildings Jungl) {reroutePins [Result {expression=&self:MaterialExpressionReroute_0}]}
  PinBase!() {reroutePins [InputPin {expression=&self:MaterialExpressionReroute_7}]}
  PinBase!(Output=@n24) {reroutePins [Output {expression=&self:MaterialExpressionReroute_6}]}
  Transform(None=@n36) {transformSourceType=TRANSFORMSOURCE_Local}
  Transform(None=@n27) {transformSourceType=TRANSFORMSOURCE_Local}
 ]
}
